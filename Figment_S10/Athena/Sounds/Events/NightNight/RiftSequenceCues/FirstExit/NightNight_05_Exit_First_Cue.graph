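# Sound cue for the fifth sound of the NightNight event's first rift exit sequence, in the high-priority events sound class.
# A mixer feeds three attenuation branches, but only the distant-attenuation branch holds a wave, the Rift Exit 05 sound.
# The two close-attenuation branches have empty inputs, so they contribute no audio.
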SoundCue /Figment_S10/Athena/Sounds/Events/NightNight/RiftSequenceCues/FirstExit/NightNight_05_Exit_First_Cue
root=Mixer_0
nodes [
 {
  id=Attenuation_0
  children [null]
  properties {
   attenuationSettings=/Figment_S10/Athena/Sounds/Events/NightNight/Attenuations/NightNight_Close_Att
   attenuationOverrides {falloffDistance=2000000}
  }
 }
 {
  id=Attenuation_1
  children [WavePlayer_2]
  properties {attenuationSettings=/Figment_S10/Athena/Sounds/Events/NightNight/Attenuations/NightNight_Distant_Att}
 }
 {
  id=Attenuation_2
  children [null]
  properties {
   attenuationSettings=&NightNight_Close_Att
   bOverrideAttenuation=true
   attenuationOverrides {falloffDistance=2000000}
  }
 }
 {id=Mixer_0, children [Attenuation_0, Attenuation_1, Attenuation_2], properties {inputVolume [1,1,1]}}
 {id=WavePlayer_2, properties {soundWaveAssetPtr=/Figment_S10/Athena/Sounds/Events/NightNight/SoundWaves/Event_NightNight_Rift_Exit_05}}
]
properties {
 maxDistance=2097152
 soundClassObject=/Figment_S10/Athena/Sounds/Events/SoundClasses/Fort_Events_Priority_ExtHigh
 volumeMultiplier=1.25
 attenuationOverrides {
  bEnableFocusInterpolation=true
  falloffDistance=1000000
  focusAttackInterpSpeed=0.5
  focusAzimuth=5
  focusReleaseInterpSpeed=0.25
  lPFFrequencyAtMax=500
  lPFRadiusMax=2000
  lPFRadiusMin=1000
  nonFocusAzimuth=30
  nonFocusVolumeAttenuation=0
  stereoSpread=2048
  attenuationShapeExtents {x=1024}
  customAttenuationCurve {editorCurveData {keys [<at 0.1 0>,<at 0.25 1>,<at 1 0>]}}
 }
}
